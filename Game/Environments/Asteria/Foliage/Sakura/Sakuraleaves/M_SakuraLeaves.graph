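# Default Lit, Nanite-enabled material for Asteria's sakura leaves that varies foliage per instance.
# A UV gradient shades a randomly blended pink tint with slight brightness jitter.
# Per-instance random blends red and blue vertex-color masks, then rounds the result to scale an offset toward object position, apparently collapsing selected leaves.
# No material output connections appear, and the texture sample is unwired.
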Material /Game/Environments/Asteria/Foliage/Sakura/Sakuraleaves/M_SakuraLeaves
graph {
 properties {bUsedWithNanite=true, pixelDepthOffsetMode=PDOM_Legacy, usedShadingModels=Default Lit}
 nodes [
  Ceil()
  Constant() {r=0.6}
  n1 = MaterialFunctionCall(UV Channel=TextureCoordinate()) {materialFunction=&LinearGradient}
  Multiply(A=@n1.UGradient)
  n2 = PerInstanceRandom()
  n3 = LinearInterpolate(Alpha=@n2)
  n4 = VertexColor()
  n5 = Multiply(A=@n2) {constB=2}
  n6 = LinearInterpolate(A=LinearInterpolate(A=@n4.R, B=@n4.R, Alpha=Saturate(None=@n5)), B=@n4.B, Alpha=Saturate(None=Subtract(A=@n5)))
  n7 = WorldPosition()
  Multiply(A=Round(None=Multiply(A=@n3, B=@n6)), B=Subtract(A=ObjectPositionWS(), B=Add(B=@n7.XYZ)))
  n8 = LinearInterpolate(Alpha=@n2) {constA=0.95, constB=1.05}
  n9 = LinearInterpolate(A=Constant3Vector() {constant=<rgba 2 0.8229 1.1554 1>}, B=Constant3Vector() {constant=<rgba 2 1.3068 1.6858 1>}, Alpha=@n2)
  Saturate(None=Multiply(A=LinearInterpolate(Alpha=@n1.UGradient) {constA=0.5, constB=1.2}, B=Multiply(A=@n8, B=@n9)))
  TextureSample()
 ]
}
